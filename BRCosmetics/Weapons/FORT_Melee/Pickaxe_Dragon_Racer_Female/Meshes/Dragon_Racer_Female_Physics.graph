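# Physics asset for the Dragon Racer Female pickaxe, a two-handed melee weapon cosmetic, giving its feather attachment simulated motion.
# A large box body covers the handle bone and a small rotated box covers the dyn_feathers bone.
# A single constraint ties the feathers to the handle, with wide, equal swing limits and a narrow twist limit, so they can sway but barely rotate about their axis.
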PhysicsAsset /BRCosmetics/Weapons/FORT_Melee/Pickaxe_Dragon_Racer_Female/Meshes/Dragon_Racer_Female_Physics
bodies {
 Melee_TwoHanded_Handle [{shapeType=Box, extentX=35.1365, extentY=34.1504, extentZ=114.3214, center=<0.0253 0 36.4228>}]
 dyn_feathers [{shapeType=Box, extentX=11.355, extentY=10.8793, extentZ=12.6394, center=<4.1495 0 0>, rotation=<rot 90 0 0>}]
}
constraints [{bone1Name=dyn_feathers, bone2Name=Melee_TwoHanded_Handle, swing1LimitDegrees=65, swing2LimitDegrees=65, twistLimitDegrees=10}]
